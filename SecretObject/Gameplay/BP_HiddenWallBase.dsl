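; Base Blueprint for a secret hidden wall that visually disappears when discovered and reacts to hits.
; Discovery state drives a "Visibility.Time" material parameter, using the sign of game time to encode visible or invisible.
; Damage and area-impact flashes cycle through numbered hit slots, writing location with timestamp and direction into material vector parameters, all cosmetic-only.
; An override wall material replaces the mesh's parent material, supporting texture data with Nanite.
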(event EventBeginPlay
  (|CallParentFunction)
  (CallFunction|UpdateVisibility (Variables|/Script/SecretObjectRuntime.FortSecretObjectComponent|GetbDiscovered (Variables|DiscoveryWall|GetDiscoveryComponent))))

(event OnUndiscovered(DiscoveryComponent)
  ; Secret not yet discovered - Wall should be visible
  (CallFunction|UpdateVisibility false))

(event OnDiscovered(DiscoveryComponent)
  ; Secret discovered! Wall should become invisible
  (CallFunction|UpdateVisibility true))

(event EventOnDamagePlayEffects
  (params (Damage :type Float%20%28single-precision%29) (DamageTags :type Gameplay%20Tag%20Container%20Structure%20%28by%20ref%29) (Momentum :type Vector :default "0, 0, 0") (HitInfo :type Hit%20Result%20Structure%20%28by%20ref%29) (InstigatedBy :type Fort%20Pawn%20Object%20Reference) (DamageCauser :type Actor%20Object%20Reference) (EffectContext :type Gameplay%20Effect%20Context%20Handle%20Structure))
  (bind (_bblockinghit _binitialoverlap _time _distance _location _impactpoint _normal _impactnormal _physmat _hitactor _hitcomponent _hitbonename _bonename _hititem _elementindex _faceindex _tracestart _traceend) (Collision|BreakHitResult HitInfo))
  (Networking|CanExecuteCosmeticEvents
    (:True
      (Variables|Default|SetLastHitComponent _hitcomponent)
      (Variables|Default|SetLastImpactPoint _impactpoint)
      (Variables|Default|SetLastImpactDir (Math|Vector|Normalize (- _tracestart _traceend)))
      (Variables|Default|SetCurrentHit (Math|Integer|%25%28Integer%29 (+ (Variables|Default|GetCurrentHit) 1) (Variables|Default|GetNumSupportedHits)))
      (CallFunction|UpdateHitVisuals (Variables|Default|GetCurrentHit) (Variables|Default|GetLastImpactPoint) (Variables|Default|GetLastImpactDir)))))

(event DiscoveryWall|EventOnHitFlash
  (params (Hit :type Hit%20Result%20Structure%20%28by%20ref%29) (HitType :type EFortRevealWallHitType%20Enum :default "DirectImpact"))
  (bind (_bblockinghit _binitialoverlap _time _distance _location _impactpoint _normal _impactnormal _physmat _hitactor _hitcomponent _hitbonename _bonename _hititem _elementindex _faceindex _tracestart _traceend) (Collision|BreakHitResult Hit))
  (Networking|CanExecuteCosmeticEvents
    (:True
      (switch Utilities|FlowControl|Switch|SwitchonEFortRevealWallHitType HitType
        (:DirectImpact
          _)
        (:Explosion
          _)
        (:AreaImpact
          ; TODO: Better "full actor flash" logic
          (Variables|Default|SetLastImpactPoint (Transformation|GetWorldLocation (Variables|Debug|GetStaticMeshComponent)))
          (Variables|Default|SetLastImpactDir "0, 0, 0")
          (Variables|Default|SetLastHitComponent _hitcomponent)
          (Variables|Default|SetCurrentHit (Math|Integer|%25%28Integer%29 (+ (Variables|Default|GetCurrentHit) 1) (Variables|Default|GetNumSupportedHits)))
          (CallFunction|UpdateHitVisuals (Variables|Default|GetCurrentHit) (Variables|Default|GetLastImpactPoint) (Variables|Default|GetLastImpactDir))))
      (Variables|Default|SetLastImpactPoint _impactpoint)
      (Variables|Default|SetLastImpactDir (Math|Vector|Normalize (- _tracestart _traceend))))))

(fn GetTextureDataParentMaterialOverride
  ; Pass Override Material so we can use TD+Nanite support but specify material instead of pulling from StaticMesh
  (return (Variables|Default|GetOverrideWallMaterial)))

(fn UpdateHitVisuals
  (params (HitIndex :type Integer) (HitLocation :type Vector) (HitDirection :type Vector))
  (bind _returnvalue (Utilities|String|ToString(Integer) HitIndex))
  (Networking|CanExecuteCosmeticEvents
    (:True
      (DiscoveryWall|SetDiscoveryVectorParameterValue (Utilities|String|StringToName (Utilities|String|Append "Hit" _returnvalue)) (Math|Color|MakeColor (.x HitLocation) (.y HitLocation) (.z HitLocation) (Utilities|Time|GetGameTimeinSeconds)))
      (DiscoveryWall|SetDiscoveryVectorParameterValue (Utilities|String|StringToName (Utilities|String|Append "HitDir" _returnvalue)) (Math|Color|MakeColor (.x HitDirection) (.y HitDirection) (.z HitDirection) (Math|Vector|VectorLength HitDirection))))))

(fn UpdateVisibility
  (params (bDiscovered :type Boolean))
  (Networking|CanExecuteCosmeticEvents
    (:True
      (DiscoveryWall|SetDiscoveryScalarParameterValue "Visibility.Time" (* (Utilities|Time|GetGameTimeinSeconds) (Math|Float|SelectFloat 1.0 -1.0 bDiscovered))))))

(fn ConstructionScript)
