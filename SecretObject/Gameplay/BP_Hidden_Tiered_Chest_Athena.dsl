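; Tiered loot chest that stays hidden until its FortSecretObject component is discovered.
; Discovery changes toggle FX, lights and dynamic override materials, encoding reveal or hide as signed game time in "Visibility.Time".
; Hit positions and times cycle through numbered material vector parameters, while damage numbers are suppressed when undiscovered and health-bar suppression toggles where cosmetics cannot run.
; Looting turns off out-of-range undiscovery on the authority.
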(event EventBeginPlay
  (|CallParentFunction)
  (CallFunction|SetupDiscoveryMaterials)
  (CallFunction|UpdateDisoveryVisuals))

(event Custom|HideChest
  (Networking|CanExecuteCosmeticEvents
    (:True
      (CallFunction|SetupDiscoveryMaterials)
      (CallFunction|UpdateMaterialVisibility false)
      (Components|Activation|Deactivate (Variables|Default|GetRare%20ChestFX))
      (Rendering|SetVisibility (Variables|Default|GetFX%20Glow%20&%20Smoke%20Sheet) false false)
      (CallFunction|SetLightVisibility false))
    (:False
      (CallFunction|SetSuppressHealthBar true))))

(event Custom|RevealChest
  (Networking|CanExecuteCosmeticEvents
    (:True
      (CallFunction|SetupDiscoveryMaterials)
      (CallFunction|UpdateMaterialVisibility true)
      (if (not (Variables|Default|GetbOpened))
        (Components|Activation|Activate (Variables|Default|GetRare%20ChestFX) false)
        (Rendering|SetVisibility (Variables|Default|GetFX%20Glow%20&%20Smoke%20Sheet) true false)
        (CallFunction|SetLightVisibility true)))
    (:False
      (CallFunction|SetSuppressHealthBar false))))

(event EventMeshSetChanged
  (params (BuildingSMActor :type Building%20Static%20Mesh%20Object%20Reference) (NewMeshSet :type Mesh%20Set%20Structure%20%28by%20ref%29))
  (|CallParentFunction BuildingSMActor NewMeshSet)
  (Networking|CanExecuteCosmeticEvents
    (:True
      (if (CallFunction|IsRareChest)
        (Utilities|Array|Clear (Variables|Default|GetOriginalMaterials))
        (Utilities|Array|Clear (Variables|Default|GetDisoveryMaterials))
        (CallFunction|SetupDiscoveryMaterials)
        (CallFunction|UpdateDisoveryVisuals)))))

(event Custom|UpdateDisoveryVisuals
  (if (Variables|/Script/SecretObjectRuntime.FortSecretObjectComponent|GetbDiscovered (Variables|Default|GetFortSecretObject))
    (CallFunction|RevealChest)
    (else
      (CallFunction|HideChest))))

(event OnDiscovered(FortSecretObject)
  (CallFunction|UpdateDisoveryVisuals))

(event OnUndiscovered(FortSecretObject)
  (CallFunction|UpdateDisoveryVisuals))

(event EventOnDamagePlayEffects
  (params (Damage :type Float%20%28single-precision%29) (DamageTags :type Gameplay%20Tag%20Container%20Structure%20%28by%20ref%29) (Momentum :type Vector :default "0, 0, 0") (HitInfo :type Hit%20Result%20Structure%20%28by%20ref%29) (InstigatedBy :type Fort%20Pawn%20Object%20Reference) (DamageCauser :type Actor%20Object%20Reference) (EffectContext :type Gameplay%20Effect%20Context%20Handle%20Structure))
  (bind (_bblockinghit _binitialoverlap _time _distance _location _impactpoint _normal _impactnormal _physmat _hitactor _hitcomponent) (Collision|BreakHitResult HitInfo))
  (Networking|CanExecuteCosmeticEvents
    (:True
      (Variables|Default|SetLastHitComponent _hitcomponent)
      (Variables|Default|SetLastImpactPoint _impactpoint)
      (Variables|Default|SetCurrentHit (Math|Integer|%25%28Integer%29 (+ (Variables|Default|GetCurrentHit) 1) (Variables|Default|GetNumSupportedHits)))
      (CallFunction|UpdateHitVisuals (Variables|Default|GetCurrentHit) (Variables|Default|GetLastImpactPoint)))))

(event EventOnLoot
  ; Once a chest is looted disable the Discovery/Invisibility logic
  ; NOTE: If we need it to continue the no-proximity-invisibility, then we'll need to re-assign the material on the swapped chest mesh
  (|CallParentFunction)
  (Networking|SwitchHasAuthority
    (:Authority
      (Variables|/Script/SecretObjectRuntime.FortSecretObjectComponent|SetbUndiscoverWhenOutOfRange false (Variables|Default|GetFortSecretObject)))
    (:Remote
      (Networking|CanExecuteCosmeticEvents))))

(event EventOnLootRepeat
  ; Once a chest is looted disable the Discovery/Invisibility logic
  ; NOTE: If we need it to continue the no-proximity-invisibility, then we'll need to re-assign the material on the swapped chest mesh
  (|CallParentFunction)
  (Networking|SwitchHasAuthority
    (:Authority
      (Variables|/Script/SecretObjectRuntime.FortSecretObjectComponent|SetbUndiscoverWhenOutOfRange false (Variables|Default|GetFortSecretObject)))
    (:Remote
      (Networking|CanExecuteCosmeticEvents))))

(fn GetTextureDataParentMaterialOverride
  (return 0))

(fn SetDiscoveryScalarMaterialParam
  (params (ParameterName :type Name) (Value :type Float%20%28double-precision%29))
  (for _array_element (Variables|Default|GetDisoveryMaterials)
    (Rendering|Material|SetScalarParameterValue _array_element ParameterName Value :!class /Script/Engine.MaterialInstanceDynamic)))

(fn SetDiscoveryVectorMaterialParam
  (params (ParameterName :type Name) (Value :type Linear%20Color%20Structure))
  (for _array_element (Variables|Default|GetDisoveryMaterials)
    (Rendering|Material|SetVectorParameterValue _array_element ParameterName Value :!class /Script/Engine.MaterialInstanceDynamic)))

(fn SetupDiscoveryMaterials
  (bind _staticmeshcomponent (Variables|Debug|GetStaticMeshComponent))
  (Networking|CanExecuteCosmeticEvents
    (:True
      (if (Utilities|Array|IsEmpty (Variables|Default|GetDisoveryMaterials))
        (Utilities|IsValid _staticmeshcomponent
          (:Is%20Valid
            (for _index (range (Rendering|Material|GetNumMaterials _staticmeshcomponent))
              (LocalVariables|SetCurrParentMaterial 0)
              (Utilities|Array|Add (Variables|Default|GetOriginalMaterials) (Rendering|Material|GetMaterial _staticmeshcomponent _index))
              (bind _asmaterial_instance_dynamic (Utilities|Casting|CastToMaterialInstanceDynamic (Rendering|Material|GetMaterial _staticmeshcomponent _index))
                (:then
                  (Utilities|IsValid _asmaterial_instance_dynamic
                    (:Is%20Valid
                      (LocalVariables|SetCurrParentMaterial (Variables|/Script/Engine.MaterialInstance|GetParent _asmaterial_instance_dynamic)))
                    (:Is%20Not%20Valid
                      (LocalVariables|SetCurrParentMaterial (Rendering|Material|GetMaterial _staticmeshcomponent _index))))
                  (Utilities|IsValid (LocalVariables|GetCurrParentMaterial)
                    (:Is%20Valid
                      (bind _returnvalue (Rendering|Material|CreateDynamicMaterialInstance (select (CallFunction|IsRareChest) (Variables|Default|GetOverrideHiddenRareChestMaterial) (Variables|Default|GetOverrideHiddenChestMaterial)) "None" "Transient" :!class /Script/Engine.KismetMaterialLibrary))
                      (Rendering|Material|CopyParameterOverrides _returnvalue _asmaterial_instance_dynamic)
                      (Rendering|Material|SetMaterial (Variables|Debug|GetStaticMeshComponent) _index _returnvalue)
                      (Utilities|Array|Add (Variables|Default|GetDisoveryMaterials) _returnvalue))
                    (:Is%20Not%20Valid
                      _)))
                (:CastFailed
                  _))
              (Utilities|Array|Add (Variables|Default|GetDisoveryMaterials)))))))))

(fn ShouldDisplayDamageNumbers
  (bind _returnvalue (|CallParentFunction))
  (if (not (Variables|/Script/SecretObjectRuntime.FortSecretObjectComponent|GetbDiscovered (Variables|Default|GetFortSecretObject)))
    (return false)
    (else
      (return _returnvalue))))

(fn SwapBackToOriginalMaterials
  (bind _staticmeshcomponent (Variables|Debug|GetStaticMeshComponent))
  (if (Utilities|Array|IsNotEmpty (Variables|Default|GetOriginalMaterials))
    (Utilities|IsValid _staticmeshcomponent
      (:Is%20Valid
        (for _index (range (Rendering|Material|GetNumMaterials _staticmeshcomponent))
          (Rendering|Material|SetMaterial (Variables|Debug|GetStaticMeshComponent) _index (Utilities|Array|Get(acopy) (Variables|Default|GetOriginalMaterials) _index)))))))

(fn UpdateHitVisuals
  (params (HitIndex :type Integer) (HitLocation :type Vector))
  (Networking|CanExecuteCosmeticEvents
    (:True
      (CallFunction|SetDiscoveryVectorMaterialParam (Utilities|String|StringToName (Utilities|String|Append "Hit" (Utilities|String|ToString(Integer) HitIndex))) (Math|Color|MakeColor (.x HitLocation) (.y HitLocation) (.z HitLocation) (Utilities|Time|GetGameTimeinSeconds))))))

(fn UpdateMaterialVisibility
  (params (bDiscovered :type Boolean))
  (Networking|CanExecuteCosmeticEvents
    (:True
      (CallFunction|SetDiscoveryScalarMaterialParam "Visibility.Time" (* (Utilities|Time|GetGameTimeinSeconds) (Math|Float|SelectFloat 1.0 -1.0 bDiscovered))))))

(fn ConstructionScript
  (|CallParentFunction))
